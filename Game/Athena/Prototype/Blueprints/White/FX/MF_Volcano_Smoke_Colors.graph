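# Material function that outputs time-of-day colors for smoke, apparently for volcano effects, lit by the sun and moon direction.
# Clamped dot of that collection-driven direction and an input normal, raised to a power, scales highlight colors blended through Color_Time_of_Day.
# Time-of-day shadow colors, scaled by a multiplier, are added on top.
# The mobile shading path skips this, lerping only night shadow and highlight colors by the MaskForLowEnd input.
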MaterialFunction /Game/Athena/Prototype/Blueprints/White/FX/MF_Volcano_Smoke_Colors
graph {nodes [
  n1 = CollectionParameter() {
   collection=&FortniteMaterialParameters
   parameterName=SunAndMoonModelDirectionalVector
  }
  n2 = DotProduct(
   A=Normalize(VectorInput=ComponentMask(None=@n1) {b=true, g=true, r=true})
   B=FunctionInput(Preview=Constant3Vector() {constant=<rgba 0 0 1 0>}) {bUsePreviewValueAsDefault=true, inputName=WorldSpaceNormal}
  )
  n3 = VectorParameter() {group=Colors, parameterName=NightH, defaultValue=<rgba 0.015 0.0225 0.025 1>}
  n4 = Multiply(
   A=Power(Base=Clamp(None=@n2), Exp=ScalarParameter() {defaultValue=2, group=Color, parameterName=HighlightPower})
   B=MaterialFunctionCall(
    Color_Morning=VectorParameter() {group=Colors, parameterName=MorningH, defaultValue=<rgba 0.03 0.0219 0.0121 1>}.RGB
    Color_Day=VectorParameter() {group=Colors, parameterName=DayH, defaultValue=<rgba 0.0317 0.0372 0.055 1>}.RGB
    Color_Evening=VectorParameter() {group=Colors, parameterName=EveningH, defaultValue=<rgba 0.1 0.0266 0.02 1>}.RGB
    Color_Night=@n3.RGB
   ) {materialFunction=&Color_Time_of_Day}.Result
  ) {constB=0}
  n5 = VectorParameter() {group=Colors, parameterName=ShadowNight, defaultValue=<rgba 0.0205 0.0232 0.025 1>}
  n6 = Multiply(
   A=MaterialFunctionCall(
    Color_Morning=VectorParameter() {group=Colors, parameterName=ShadowMorning, defaultValue=<rgba 0.0271 0.0277 0.04 1>}.RGB
    Color_Day=VectorParameter() {group=Colors, parameterName=ShadowDay, defaultValue=<rgba 0.0171 0.0181 0.025 1>}.RGB
    Color_Evening=VectorParameter() {group=Colors, parameterName=ShadowEvening, defaultValue=<rgba 0.0378 0.0417 0.05 1>}.RGB
    Color_Night=@n5.RGB
   ) {materialFunction=&Color_Time_of_Day}.Result
   B=ScalarParameter() {defaultValue=0.4, group=Color, parameterName=ShadowMult}
  ) {constB=0.25}
  n7 = FunctionInput(Preview=Constant()) {bUsePreviewValueAsDefault=true, inputName=MaskForLowEnd, inputType=FunctionInput_Scalar, sortPriority=1}
  FunctionOutput(None=ShadingPathSwitch(Default=Add(A=@n4, B=@n6), Mobile=LinearInterpolate(A=@n5.RGB, B=@n3.RGB, Alpha=@n7)) {desc=Migrated from FeatureLevelSwitch})
 ]}
