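# Material function that colors a giant volcano smoke pillar, summing emissive glow, sunlight scattering and darkened smoke color.
# Glow blends two colors using a height mask (pivot sphere in Creative), normal-versus-glow direction and inverse input mask, boosted by MPC_Snow's SmokeGlowAdd and time of day.
# Sunlight uses custom Schlick phase, time-of-day tint and Fresnel, zeroed on low quality.
# MF_Volcano_Smoke_Colors supplies smoke color, dimmed by mask and inner glow.
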MaterialFunction /Game/Athena/Prototype/Blueprints/White/FX/MF_Giant_Smoke_Pillar_Color
graph {nodes [
  n1 = WorldPosition()
  n2 = MaterialFunctionCall() {materialFunction=&ObjectPivotPoint}
  n3 = ScalarParameter() {defaultValue=50000, group=Color, parameterName=GlowMaskRadius}
  n4 = CollectionParameter() {collection=/Game/Athena/Prototype/Blueprints/White/MPC_Snow, parameterName=SmokeGlowAdd}
  n5 = Add(A=@n3, B=Multiply(A=Multiply(A=@n3) {constB=0.4}, B=@n4))
  n6 = FunctionInput(Preview=ComponentMask(None=WorldPosition().XYZ) {b=true}) {bUsePreviewValueAsDefault=true, inputName=GlowWorldPosition}
  n7 = FunctionInput(Preview=ScalarParameter() {defaultValue=16300, parameterName=VolcanoRimHeight}) {bUsePreviewValueAsDefault=true, inputName=GlowStartHeight, inputType=FunctionInput_Scalar}
  n8 = Power(
   Base=StaticSwitchParameter(
    True=SphereMask(A=@n1.XYZ, B=@n2.Object Pivot Location, Radius=@n5) {hardnessPercent=0}
    False=SphereMask(A=@n6, B=@n7, Radius=@n5) {hardnessPercent=0}
   ) {parameterName=IsCreative?}
   Exp=ScalarParameter() {defaultValue=2.5, group=Color, parameterName=GlowMaskPower}
  )
  n9 = ScalarParameter() {defaultValue=0.5, group=Color, parameterName=GlowMaskAdd}
  n10 = Multiply(A=@n4) {constB=0.2}
  n11 = FunctionInput(Preview=Constant3Vector() {constant=<rgba 0 0 1 0>}) {bUsePreviewValueAsDefault=true, inputName=Normal}
  n12 = FunctionInput(Preview=VectorParameter() {parameterName=GlowDirection, defaultValue=<rgba 0 0 -0.5 1>}.RGB) {bUsePreviewValueAsDefault=true, inputName=GlowDirection}
  n13 = Subtract(A=VectorParameter() {parameterName=VolcanoPosition, defaultValue=<rgba 84500 36944 16300 0>}.RGB, B=WorldPosition().XYZ)
  n14 = FunctionInput(Preview=ScalarParameter() {parameterName=GlowDirectionBlend}) {bUsePreviewValueAsDefault=true, inputName=GlowDirectionBlend, inputType=FunctionInput_Scalar, sortPriority=4}
  n15 = DotProduct(A=@n11, B=Normalize(VectorInput=LinearInterpolate(A=@n12, B=@n13, Alpha=@n14) {constAlpha=0}))
  n16 = Add(A=Multiply(A=@n4) {constB=0.5})
  n17 = FunctionInput(Preview=Constant()) {bUsePreviewValueAsDefault=true, inputName=Mask, sortPriority=1}
  n18 = OneMinus(None=@n17)
  n19 = Multiply(
   A=FunctionInput(Preview=ScalarParameter() {defaultValue=1, group=Color, parameterName=InnerGlowMult}) {bUsePreviewValueAsDefault=true, inputName=Inner Glow Mult, sortPriority=4}
   B=Multiply(A=@n18, B=@n18)
  )
  n20 = Add(A=Multiply(A=@n8, B=Add(A=@n9, B=@n10)) {constB=0.5}, B=Add(A=Saturate(None=@n15), B=Multiply(A=@n16, B=@n19))) {constB=0.2}
  n21 = Power(Base=@n20, Exp=ScalarParameter() {defaultValue=3, group=Color, parameterName=GlowMaskPower2}) {constExponent=3}
  n22 = Multiply(
   A=LinearInterpolate(
    A=VectorParameter() {group=Color, parameterName=GlowColor2, defaultValue=<rgba 1 0.01 0 0>}.RGB
    B=VectorParameter() {group=Color, parameterName=GlowColor1, defaultValue=<rgba 3 0.731 0 0>}.RGB
    Alpha=Saturate(None=@n21)
   )
   B=@n20
  )
  n23 = Multiply(A=@n8, B=@n22)
  n24 = Multiply(A=@n22, B=Add(A=ScalarParameter() {defaultValue=0.06, group=Color, parameterName=UpperGlowMult}, B=Multiply(A=@n4) {constB=0.1})) {constB=0.1}
  n25 = Add(A=@n4, B=ScalarParameter() {defaultValue=1, group=Color, parameterName=BaseGlow})
  n26 = Constant() {r=1.5}
  n27 = Constant() {r=1}
  n28 = Multiply(
   A=Multiply(
    A=Add(A=@n23, B=@n24)
    B=FunctionInput(Preview=@n25) {bUsePreviewValueAsDefault=true, inputName=Base Glow Mult, inputType=FunctionInput_Scalar, sortPriority=3}
   )
   B=MaterialFunctionCall(Value_Morning=@n26, Value_Day=@n27, Value_Evening=@n26, Value_Night=@n27) {materialFunction=&Scalar_Time_of_Day}.Result
  )
  n29 = Constant()
  n30 = CollectionParameter() {
   collection=&FortniteMaterialParameters
   parameterName=SunlightDirectionalVector
  }
  MaterialExpressionDotProduct_0 = DotProduct(A=ComponentMask(None=@n30) {b=true, g=true, r=true}, B=CameraVectorWS())
  MaterialExpressionScalarParameter_4 = ScalarParameter() {defaultValue=0.4, parameterName=Anisotropy}
  n31 = Custom!(CosTheta=@MaterialExpressionDotProduct_0, k=@MaterialExpressionScalarParameter_4) {
   code="\tfloat Inner = (1 + k * CosTheta);\r\n\treturn (1 - k * k) / (4 * PI * Inner * Inner);\r\n\r\n"
   description=Schlick Phase
   inputs [
    {inputName=CosTheta, input {expression=&self:MaterialExpressionDotProduct_0}}
    {inputName=k, input {expression=&self:MaterialExpressionScalarParameter_4}}
   ]
  }
  n32 = VectorParameter() {group=Color, parameterName=SunThroughColorMorning, defaultValue=<rgba 0.5 0.3174 0.2365 0>}
  n33 = MaterialFunctionCall(
   Color_Morning=@n32.RGB
   Color_Day=VectorParameter() {group=Color, parameterName=SunThroughColorDay, defaultValue=<rgba 0.5 0.3174 0.2365 0>}.RGB
   Color_Evening=@n32.RGB
   Color_Night=VectorParameter() {group=Color, parameterName=SunThroughColorNight, defaultValue=<rgba 0.5 0.3174 0.2365 0>}.RGB
  ) {materialFunction=&Color_Time_of_Day}
  n34 = VertexNormalWS()
  n35 = Fresnel(
   ExponentIn=ScalarParameter() {defaultValue=4, parameterName=LightThroughCloudsFresnelXP}
   Normal=Multiply(A=LinearInterpolate(A=@n34, B=@n11), B=TwoSidedSign())
  ) {baseReflectFraction=0.02}
  n36 = OneMinus(None=@n35)
  n37 = ScalarParameter() {parameterName=InvertLightThroughCloudsFresnel}
  n38 = QualitySwitch(Default=Multiply(A=Multiply(A=@n31, B=@n33.Result), B=LinearInterpolate(A=@n35, B=@n36, Alpha=@n37)), Low=Constant())
  n39 = ScalarParameter() {defaultValue=0.5, group=Color, parameterName=MinSmokeBrightness}
  n40 = Multiply(
   A=OneMinus(None=Saturate(None=@n19))
   B=Multiply(
    A=MaterialFunctionCall!(WorldSpaceNormal=@n11, MaskForLowEnd=@n17) {materialFunction=&/MF_Volcano_Smoke_Colors}.Result
    B=LinearInterpolate(A=@n39, Alpha=@n17) {constA=0.15}
   )
  )
  FunctionOutput(None=Add(A=StaticSwitchParameter(True=@n28, False=@n29) {defaultValue=true, parameterName=UseBaseGlow}, B=Add(A=@n38, B=@n40)))
 ]}
