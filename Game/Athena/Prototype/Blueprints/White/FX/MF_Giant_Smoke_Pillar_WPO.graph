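# World position offset function for a giant smoke pillar, apparently rising from a volcano, that widens and sways it.
# World Z, normalized between rim and top height parameters, gives a height mask for widening and phases time-scrolled sine waves.
# Widening, waves and a texture mask input combine into a scalar offset, also output along scaled vertex normals.
# A separate drift output rotates a sine-driven vector about world Z, damped toward the rim.
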MaterialFunction /Game/Athena/Prototype/Blueprints/White/FX/MF_Giant_Smoke_Pillar_WPO
graph {nodes [
  n1 = FunctionInput(Preview=Constant() {r=0.5}) {bUsePreviewValueAsDefault=true, inputName=Texture WPO Mask, inputType=FunctionInput_Scalar, sortPriority=1}
  n2 = ScalarParameter() {defaultValue=500, parameterName=TextureWPO}
  n3 = Multiply(
   A=FunctionInput(Preview=ScalarParameter() {defaultValue=0.01, parameterName=WPODriftSpeed}) {bUsePreviewValueAsDefault=true, inputName=WPO Drift Speed, inputType=FunctionInput_Scalar}
   B=Time()
  )
  n4 = ScalarParameter() {defaultValue=1.25, parameterName=WPOTile}
  n5 = WorldPosition()
  n6 = ScalarParameter() {defaultValue=50000, parameterName=TopPosition}
  n7 = Divide(
   A=Subtract(A=ComponentMask(None=@n5.XYZ) {b=true}, B=@n6)
   B=Distance(A=@n6, B=ScalarParameter() {defaultValue=16300, parameterName=VolcanoRimHeight})
  )
  n8 = Saturate(None=ConstantBiasScale(None=@n7))
  n9 = Add(A=@n3, B=Multiply(A=FunctionInput(Preview=@n4) {bUsePreviewValueAsDefault=true, inputName=WPO Tile, inputType=FunctionInput_Scalar}, B=@n8))
  n10 = ScalarParameter() {defaultValue=15000, parameterName=WPOTopWidthMult}
  n11 = Multiply(A=Sine(None=@n9) {period=0.58}, B=Multiply(A=ScalarParameter() {defaultValue=0.07, parameterName=WidenWaveMult}, B=@n10) {constB=0.1})
  n12 = LinearInterpolate(Alpha=Saturate(None=Abs(None=@n7))) {constA=1, constB=0.15}
  n13 = Power(Base=@n8, Exp=ScalarParameter() {defaultValue=1.5, parameterName=HeightMaskPower}) {constExponent=3}
  n14 = Add(
   A=Multiply(
    A=LinearInterpolate(Alpha=@n1) {constA=-1}
    B=FunctionInput(Preview=@n2) {bUsePreviewValueAsDefault=true, inputName=Texture WPO Scale, inputType=FunctionInput_Scalar, sortPriority=1}
   )
   B=Add(A=Multiply(A=@n12, B=@n11), B=Multiply(A=@n10, B=@n13))
  )
  n15 = VertexNormalWS()
  n16 = VectorParameter() {parameterName=VertNormalMult, defaultValue=<rgba 1 1 0 1>}
  FunctionOutput(None=Multiply(A=@n14, B=Multiply(A=@n15, B=@n16.RGB))) {outputName=Vert Normal WPO, sortPriority=1}
  n17 = Constant()
  n18 = Sine(None=@n9) {period=0.72}
  n19 = Sine(None=@n9) {period=0.67}
  n20 = MaterialFunctionCall!(Rotation Amount=Multiply(A=@n3) {constB=-1.2}, PivotPoint=@n17, WorldPosition=AppendVector(A=AppendVector(A=@n18, B=@n19), B=@n17)) {materialFunction=/Engine/Functions/Engine_MaterialFunctions02/WorldPositionOffset/RotateAboutWorldAxis_cheap}
  n21 = FunctionInput(Preview=ScalarParameter() {defaultValue=1500, parameterName=WPODrift}) {bUsePreviewValueAsDefault=true, inputName=WPO Drift Scale, inputType=FunctionInput_Scalar}
  FunctionOutput(None=Multiply(A=@n12, B=Multiply(A=@n20.Z-Axis, B=@n21))) {outputName=Drift WPO}
  FunctionOutput(None=@n8) {outputName=HeightMask, sortPriority=3}
  FunctionOutput(None=@n14) {outputName=NonVertNormalWPO, sortPriority=2}
 ]}
