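# Translucent unlit material for a giant smoke pillar, apparently above a volcano, with a TopMesh switch that blends in a second smoke texture.
# Panned, bump-offset cloud textures feed opacity, normals and world-position-offset drift, with fresnel, view-angle, depth and near-camera fades softening the shape.
# MPC_Snow's SmokePulse moves a height-based sphere mask up the pillar, swelling vertices and boosting inner glow.
# A shared color function applies height-graded glow, whose inner detail drops on low quality.
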Material /Game/Athena/Prototype/Blueprints/White/FX/M_Giant_Smoke_Pillar_Translucent
graph {
 properties {blendMode=BLEND_Translucent, pixelDepthOffsetMode=PDOM_Legacy, shadingModel=MSM_Unlit, usedShadingModels=Unlit}
 nodes [
  n1 = ScalarParameter() {defaultValue=3, parameterName=TileX2}
  n2 = ScalarParameter() {defaultValue=4, parameterName=TileY2}
  n3 = Multiply(A=Time(), B=ScalarParameter() {defaultValue=1, parameterName=PanSpeed})
  n4 = Panner(Coordinate=Multiply(A=TextureCoordinate(), B=AppendVector(A=@n1, B=@n2)), Time=@n3) {speedX=-0.04, speedY=0.06}
  n5 = TextureSample(UVs=@n4) {samplerType=SAMPLERTYPE_Grayscale, texture=/Game/Effects/Fort_Effects/Textures/Smoke/T_Smoke_Trail_Soft_NoClamp}
  n6 = Multiply(A=@n5.G) {constB=5}
  n7 = LinearInterpolate(Alpha=ComponentMask(None=TextureCoordinate()) {g=true}) {constA=0.2}
  n8 = Multiply(
   A=TextureCoordinate()
   B=AppendVector(A=ScalarParameter() {defaultValue=2, parameterName=TileX}, B=ScalarParameter() {defaultValue=1.5, parameterName=TileY})
  )
  n9 = Panner(Coordinate=@n8, Time=@n3) {speedX=-0.03, speedY=0.07}
  n10 = TextureSample(UVs=@n9) {samplerType=SAMPLERTYPE_LinearGrayscale, texture=/Game/VisualThreatManager/StormVisuals/LargeClouds/T_StormCloudDetail01_M}
  n11 = MaterialFunctionCall!(
   WPO Tile
   WPO Drift Speed
   WPO Drift Scale
   Texture WPO Scale
   Texture WPO Mask=StaticSwitchParameter(True=LinearInterpolate(A=@n6, B=@n10.G, Alpha=@n7), False=@n10.G) {parameterName=TopMesh}
  ) {materialFunction=&/MF_Giant_Smoke_Pillar_WPO}
  n12 = CollectionParameter() {collection=/Game/Athena/Prototype/Blueprints/White/MPC_Snow, parameterName=SmokePulse}
  n13 = OneMinus(None=@n12)
  n14 = Multiply(A=ScalarParameter() {defaultValue=4, parameterName=PulseScale}, B=Power(Base=@n13))
  n15 = ComponentMask(None=ActorPositionWS()) {b=true}
  n16 = ScalarParameter() {defaultValue=-12000, parameterName=PulseStartOffset}
  n17 = ScalarParameter() {defaultValue=45000, parameterName=PulseEndOffset}
  n18 = ScalarParameter() {defaultValue=20000, parameterName=PulseRadius}
  n19 = SphereMask(
   A=ComponentMask(None=WorldPosition().XYZ) {b=true}
   B=Add(A=Add(A=@n15, B=@n16), B=Multiply(A=@n12, B=@n17))
   Radius=LinearInterpolate(A=Multiply(A=@n18) {constB=0.35}, B=@n18, Alpha=@n12)
  ) {hardnessPercent=50}
  Add(A=Add(A=@n11.Drift WPO, B=@n11.Drift WPO), B=Multiply(A=@n11.Vert Normal WPO, B=Multiply(A=@n14, B=@n19)))
  Constant() {r=1}
  Divide(A=Time() {bOverride_Period=true, period=2})
  n20 = ScalarParameter() {defaultValue=15000, parameterName=NearFadeDistance}
  n21 = ScalarParameter() {defaultValue=0.5, parameterName=TopAngleFade}
  n22 = ScalarParameter() {defaultValue=2, parameterName=TopOpacityBoost}
  n23 = ComponentMask(None=TextureCoordinate()) {g=true}
  n24 = Add(A=@n22, B=Multiply(A=@n23, B=@n22) {constB=2}) {constA=2}
  n25 = BumpOffset(
   Coordinate=@n9
   Height=LinearInterpolate(Alpha=@n10.G) {constA=-1}
   HeightRatioInput=ScalarParameter() {defaultValue=0.07, parameterName=BumpOffsetHeightRatio}
  )
  n26 = TextureSample(UVs=@n25) {samplerType=SAMPLERTYPE_LinearGrayscale, texture=&T_StormCloudDetail01_M}
  n27 = LinearInterpolate(A=@n5.G, B=@n26.G, Alpha=@n7)
  n28 = Constant3Vector() {constant=<rgba 0 0 1 0>}
  n29 = LinearInterpolate(
   A=ScalarParameter() {defaultValue=2, parameterName=MaxFresnelExp}
   B=ScalarParameter() {defaultValue=0.75, parameterName=MinFresnelXP}
   Alpha=@n26.G
  ) {constA=5, constB=0.5}
  n30 = Multiply(
   A=ScalarParameter() {defaultValue=4, parameterName=FadeAngleMult}
   B=Power(
    Base=MaterialFunctionCall(Axis=@n28, Exponent=@n29) {materialFunction=&AxisAlignedFresnel}.Fresnel
    Exp=ScalarParameter() {defaultValue=1, parameterName=FresnelPower}
   )
  )
  n31 = LinearInterpolate(
   A=Multiply(
    A=MaterialFunctionCall("FadeSensitivity (optional)"=@n21) {materialFunction=&FadeBasedOnViewAngle}.Result
    B=Multiply(A=@n27, B=@n24) {constB=2}
   )
   B=@n30
   Alpha=@n23
  ) {constA=1}
  n32 = ScalarParameter() {defaultValue=500, parameterName=FadeDistance}
  n33 = Multiply(A=@n32) {constB=0.25}
  n34 = StaticSwitchParameter(
   True=Power(Base=Saturate(None=@n31), Exp=LinearInterpolate(Alpha=@n23) {constA=0.8, constB=3})
   False=DepthFade(Opacity=Saturate(None=@n30), FadeDistance=LinearInterpolate(A=@n32, B=@n33, Alpha=@n26.G))
  ) {parameterName=TopMesh}
  n35 = StaticSwitchParameter(True=@n27, False=@n26.G) {parameterName=TopMesh}
  n36 = LinearInterpolate(A=ScalarParameter() {defaultValue=1, parameterName=MinOpacity}, Alpha=@n35)
  Multiply(
   A=Multiply(
    A=VertexColor().R
    B=MaterialFunctionCall(Distance=@n20) {materialFunction=&Near_Camera_Fade}.Result
   )
   B=Multiply(A=ScalarParameter() {defaultValue=1, parameterName=Opacity}, B=Multiply(A=@n34, B=@n36))
  )
  n37 = Multiply(
   A=TextureSample(UVs=@n4) {samplerType=SAMPLERTYPE_Normal, texture=/Game/Effects/Fort_Effects/Textures/Smoke/T_Smoke_Trail_Soft_N}.RGB
   B=VectorParameter() {parameterName=Normal2Mult, defaultValue=<rgba 0.75 -0.75 1 1>}.RGB
  )
  n38 = TextureSample(UVs=@n25) {samplerType=SAMPLERTYPE_Normal, texture=/Game/VisualThreatManager/StormVisuals/LargeClouds/T_StormCloudDetail01_N}
  n39 = Multiply(
   A=VectorParameter() {parameterName=NormalMult, defaultValue=<rgba 1 1 1 0>}.RGB
   B=StaticSwitchParameter(True=LinearInterpolate(A=@n37, B=@n38.RGB, Alpha=@n7), False=@n38.RGB) {parameterName=TopMesh}
  )
  n40 = ScalarParameter() {defaultValue=75000, parameterName=TopTopPosition}
  n41 = Subtract(A=ComponentMask(None=WorldPosition().XYZ) {b=true}, B=@n40)
  n42 = Distance(A=@n40, B=ScalarParameter() {defaultValue=16300, parameterName=VolcanoRimHeight})
  n43 = Saturate(None=ConstantBiasScale(None=Divide(A=@n41, B=@n42)))
  n44 = Power(Base=OneMinus(None=@n43))
  n45 = TextureSample(UVs=Panner(Coordinate=Multiply(A=@n8) {constB=2}, Time=@n3) {speedX=0.02, speedY=0.25}) {samplerType=SAMPLERTYPE_LinearGrayscale, texture=&T_StormCloudDetail01_M}
  n46 = Multiply(A=@n45.G, B=@n45.G)
  n47 = Multiply(A=@n13, B=@n19)
  n48 = Multiply(A=Multiply(A=Multiply(A=@n46) {constB=2}, B=@n44), B=LinearInterpolate(Alpha=Saturate(None=@n47)) {constA=1, constB=5})
  n49 = Constant()
  Reroute(None=MaterialFunctionCall!(
    GlowDirection
    GlowStartHeight
    GlowWorldPosition
    Normal=Transform(None=@n39)
    Mask=@n35
    Base Glow Mult=LinearInterpolate(Alpha=@n44) {constA=0.1}
    Inner Glow Mult=QualitySwitch(Default=@n48, Low=@n49)
    GlowDirectionBlend=@n43
   ) {materialFunction=&/MF_Giant_Smoke_Pillar_Color}.Result)
 ]
}
